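# Looping glider thrust sound cue for the LazarusLens glider cosmetic.
# A mixer layers a thrust loop and a sparkle loop at equal volume.
# Only the thrust loop, behind a single-entry random node, passes through a continuous modulator that maps the AnyDirection parameter to pitch and volume, while the sparkle plays unmodulated.
# Glider thrust attenuation, sound class and cosmetic loop property sheet govern playback.
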SoundCue /BRCosmetics/Sounds/Gliders/LazarusLens/GliderThrustLoop_LazarusLens_Cue1
root=Mixer_0
nodes [
 {id=Mixer_0, children [ModulatorContinuous_0, WavePlayer_1], properties {inputVolume [1,1]}}
 {
  id=ModulatorContinuous_0
  children [Random_0]
  properties {
   pitchModulationParams {default=0, maxOutput=1.1, minOutput=0.75, parameterName=AnyDirection}
   volumeModulationParams {default=0, maxOutput=1.2, minOutput=0.5, parameterName=AnyDirection}
  }
 }
 {id=Random_0, children [WavePlayer_0], properties {weights [1]}}
 {id=WavePlayer_0, properties {bLooping=true, soundWaveAssetPtr=/Game/Sounds/Fort_Gadgets/Athena/LazarusLens/Glider_LazarusLens_ThrustLoop}}
 {id=WavePlayer_1, properties {bLooping=true, soundWaveAssetPtr=&/Glider_LazarusLens_SparkleLoop}}
]
properties {
 attenuationSettings=&Fort_Player_Glider_Thrust_Athena_Att
 audioPropertiesSheet=&APS_Cosmetic_Glider_Loop
 maxDistance=12500
 soundClassObject=&Fort_Gliders_Land
 volumeMultiplier=1
 concurrencySet [null]
}
